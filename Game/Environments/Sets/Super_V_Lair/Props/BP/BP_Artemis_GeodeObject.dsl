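; Geode prop Blueprint for the Super V Lair set that turns on vent smoke as a calendar lead-up event nears its end.
; After a short BeginPlay delay it waits for match ready on authority, then watches the LeadUp calendar event.
; When the event becomes active it stores the seconds remaining and, at thirty minutes or less, sets B_Smoke.
; The B_Smoke RepNotify activates the vent smoke component, while the IsRocketFinal RepNotify is empty.
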(event EventBeginPlay
  (Utilities|FlowControl|Delay 3.0)
  (Class|FortAsyncActionAuthorityMatchReady|MatchReadyAsync
    (:OnReady
      (bind (_eventname _timeuntilend) (MCPContext|WatchCalendarEvent (Variables|Default|GetLeadUp))
        (:OnEventActive
          (bind _output_get (Variables|Default|SetTotalSecondsRemaining (Math|Timespan|GetTotalSeconds _timeuntilend)))
          (if (<= _output_get 1.8e+03)
            (Variables|Default|SetB_Smoke true)))))))

(fn OnRep_B_Smoke
  (if (Variables|Default|GetB_Smoke)
    (Components|Activation|Activate (Variables|Default|GetP_Geode_1_Vents_Smoke) false)))

(fn OnRep_IsRocketFinal)

(fn ConstructionScript)
